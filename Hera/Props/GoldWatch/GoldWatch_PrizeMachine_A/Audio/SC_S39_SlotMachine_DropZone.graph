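# Sound cue that plays the looping slot machine drop zone music for the GoldWatch prize machine prop.
# A random Delay of up to five seconds feeds a single looping Wave Player, so the start time varies per play.
# It routes to the POI music sound class at reduced volume, with shared casino slot machine attenuation and two concurrency sets, one drop-zone-specific and one for slot machines.
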SoundCue /Hera/Props/GoldWatch/GoldWatch_PrizeMachine_A/Audio/SC_S39_SlotMachine_DropZone
root=Delay_0
nodes [
 {id=Delay_0, children [WavePlayer_0], properties {delayMax=5}}
 {
  id=WavePlayer_0
  properties {bLooping=true, soundWaveAssetPtr=&/MX_S39_SlotMachine_DropZone_LP}
 }
]
properties {
 attenuationSettings=&/ATT_Casino_SlotMachines
 maxDistance=3000
 soundClassObject=/Game/Sounds/FortSoundClassesAndModes/Fort_Music_POI_PSM
 volumeMultiplier=0.4
 assetUserData [&self:AudioAssetUserData_0]
 concurrencySet [
  &/CONC_SlotMachine_DropZone
  &/CONC_SlotMachines
 ]
}
