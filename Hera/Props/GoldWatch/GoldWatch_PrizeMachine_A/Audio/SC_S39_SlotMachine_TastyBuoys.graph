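# Sound cue for the looping Tasty Buoys slot machine music on the GoldWatch prize machine.
# A single looping wave player feeds the root directly, with reduced volume and a POI music sound class.
# Casino slot machine attenuation spatializes it, while machine-specific and shared slot machine concurrency groups limit overlapping instances.
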SoundCue /Hera/Props/GoldWatch/GoldWatch_PrizeMachine_A/Audio/SC_S39_SlotMachine_TastyBuoys
root=WavePlayer_0
nodes [{
  id=WavePlayer_0
  properties {bLooping=true, soundWaveAssetPtr=&/MX_S39_SlotMachine_TastyBuoys_LP}
 }]
properties {
 attenuationSettings=&/ATT_Casino_SlotMachines
 maxDistance=3000
 soundClassObject=/Game/Sounds/FortSoundClassesAndModes/Fort_Music_POI_PSM
 volumeMultiplier=0.4
 assetUserData [&self:AudioAssetUserData_0]
 concurrencySet [
  &/CONC_SlotMachine_TastyBuoys
  &/CONC_SlotMachines
 ]
}
