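; Gold Watch prize machine prop that randomizes a secondary swatch and matching music loop at construction.
; A location-seeded random stream picks from the current palette, stepping to the next entry, wrapping around, when it matches the primary swatch index.
; The swatch's override material fills mesh slot 0; otherwise a map supplies slot 1, with a screen material fallback.
; An audio map sets the slot machine music for that swatch.
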(event EventBeginPlay)

(event Collision|EventActorBeginOverlap
  (params (OtherActor :type Actor%20Object%20Reference)))

(event EventTick
  (params (DeltaSeconds :type Float%20%28single-precision%29)))

(fn AssignNewMaterialsToSecondaryMesh
  (params (TextureData :type Building%20Texture%20Data%20Object%20Reference) (MeshComponent :type Primitive%20Component%20Object%20Reference))
  (bind _l_texturedata (LocalVariables|GetL_TextureData))
  (bind _materialmap (Variables|Default|GetMaterialMap))
  (bind (_value _returnvalue) (Utilities|Map|Find _materialmap _l_texturedata))
  (bind _l_texturedata_2 (LocalVariables|GetL_TextureData))
  (bind _overridematerial (Variables|/Script/FortniteGame.BuildingTextureData|GetOverrideMaterial _l_texturedata_2))
  (LocalVariables|SetL_TextureData TextureData)
  (if (Utilities|IsValid _overridematerial :!class /Script/Engine.KismetSystemLibrary)
    (Rendering|Material|SetMaterial MeshComponent 0 _overridematerial)
    (if _returnvalue
      (Rendering|Material|SetMaterial MeshComponent 1 _value)
      (else
        (Rendering|Material|SetMaterial MeshComponent 1 "/Hera/Props/GoldWatch/GoldWatch_PrizeMachine_A/Materials/MI_GoldWatch_PrizeMachineScreenDrop_A.MI_GoldWatch_PrizeMachineScreenDrop_A")))))

(fn GetRandomStream
  (bind _returnvalue (Math|Random|MakeRandomStream (Math|Float|Truncate (* (Class|EnvironmentArtFuncLib|RandomFloatbyLocation (Transformation|GetActorLocation) (Variables|RandomSwatch|GetSwatchTransitionDistance) true) 1e+04))))
  (Variables|Default|SetRandomStream _returnvalue))

(fn RandomizeSecondarySwatch
  (params (TextureDataArray :type Array%20of%20Building%20Texture%20Data%20Object%20References))
  (bind _randomstream (Variables|Default|GetRandomStream))
  (bind _l_ar_td (LocalVariables|GetL_AR_TD))
  (bind (_outitem _outindex) (Utilities|Array|RandomArrayItemfromStream _l_ar_td _randomstream))
  (bind _swatchindex (Variables|RandomSwatch|GetSwatchIndex))
  (bind _returnvalue (+ _swatchindex 1))
  (LocalVariables|SetL_AR_TD TextureDataArray)
  (if (!= _outindex _swatchindex)
    (return _outitem)
    (elif (<= _returnvalue (- (Utilities|Array|Length (LocalVariables|GetL_AR_TD)) 1))
      (return (Utilities|Array|Get(acopy) (LocalVariables|GetL_AR_TD) _returnvalue))
      (else
        (return (Utilities|Array|Get(acopy) (LocalVariables|GetL_AR_TD)))))))

(fn ConstructionScript
  (bind _audiomap_0 (Variables|Default|GetAudioMap_0))
  (|CallParentFunction)
  (CallFunction|GetRandomStream)
  (bind _randomizedtexturedata (CallFunction|RandomizeSecondarySwatch (Variables|/Game/Environments/Apollo/Blueprints/RandomColor_Prop/RandomColor_Prop_Palette.RandomColor_Prop_Palette_C|GetSwatch_Palette (Variables|Default|GetcurrentPalette))))
  (CallFunction|AssignNewMaterialstoSecondaryMesh _randomizedtexturedata (Variables|Default|GetSM_GoldWatch_PrizeMachine_A))
  (if (.ReturnValue (Utilities|Map|Find _audiomap_0 _randomizedtexturedata))
    (if (not (== (Variables|/Game/Sounds/Fort_Audio_Blueprints/PropAmbientAudioComponent/FortPropAudioComponent.FortPropAudioComponent_C|GetSoundLoop (Variables|Default|GetSlotMachineMusic)) (Utilities|Map|Find _audiomap_0 _randomizedtexturedata)))
      (Variables|/Game/Sounds/Fort_Audio_Blueprints/PropAmbientAudioComponent/FortPropAudioComponent.FortPropAudioComponent_C|SetSoundLoop (Utilities|Map|Find _audiomap_0 _randomizedtexturedata) (Variables|Default|GetSlotMachineMusic_2)))))
