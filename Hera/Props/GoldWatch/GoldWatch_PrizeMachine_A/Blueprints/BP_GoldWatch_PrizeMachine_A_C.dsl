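; Prize machine prop Blueprint that randomizes the screen materials and slot-machine music of its two meshes during construction.
; A location-seeded random stream picks palette swatches, avoiding the current swatch index and removing the first pick before choosing the second.
; Each pick applies its override and mapped materials, with a fallback screen material.
; An audio map keyed by swatch sets extra music loops, checking against loops already assigned to avoid duplicates.
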(event EventBeginPlay
  (|CallParentFunction) ; This node is disabled and will not be called.\nDrag off pins to build functionality.
)

(event Collision|EventActorBeginOverlap
  (params (OtherActor :type Actor%20Object%20Reference))
  (|CallParentFunction OtherActor) ; This node is disabled and will not be called.\nDrag off pins to build functionality.
)

(event EventTick
  (params (DeltaSeconds :type Float%20%28single-precision%29))
  (|CallParentFunction DeltaSeconds) ; This node is disabled and will not be called.\nDrag off pins to build functionality.
)

(fn AssignNewMaterialsToSecondaryMesh
  (params (TextureData :type Building%20Texture%20Data%20Object%20Reference) (MeshComponent :type Primitive%20Component%20Object%20Reference))
  (bind _l_texturedata (LocalVariables|GetL_TextureData))
  (bind _materialmap (Variables|Default|GetMaterialMap))
  (bind (_value _returnvalue) (Utilities|Map|Find _materialmap _l_texturedata))
  (bind _l_texturedata_2 (LocalVariables|GetL_TextureData))
  (bind _overridematerial (Variables|/Script/FortniteGame.BuildingTextureData|GetOverrideMaterial _l_texturedata_2))
  (LocalVariables|SetL_TextureData TextureData)
  (if (Utilities|IsValid _overridematerial :!class /Script/Engine.KismetSystemLibrary)
    (Rendering|Material|SetMaterial MeshComponent 0 _overridematerial)
    (if _returnvalue
      (Rendering|Material|SetMaterial MeshComponent 1 _value)
      (else
        (Rendering|Material|SetMaterial MeshComponent 1 "/Hera/Props/GoldWatch/GoldWatch_PrizeMachine_A/Materials/MI_GoldWatch_PrizeMachineScreenDrop_A.MI_GoldWatch_PrizeMachineScreenDrop_A")))))

(fn GetRandomStream
  (bind _returnvalue (Math|Random|MakeRandomStream (Math|Float|Truncate (* (Class|EnvironmentArtFuncLib|RandomFloatbyLocation (Transformation|GetActorLocation) (Variables|RandomSwatch|GetSwatchTransitionDistance) true) 1e+04))))
  (Variables|Components|SetRandomStream _returnvalue))

(fn RandomizeSecondarySwatch
  (params (TextureDataArray :type Array%20of%20Building%20Texture%20Data%20Object%20References))
  (bind _randomstream (Variables|Components|GetRandomStream))
  (bind _l_ar_td (LocalVariables|GetL_AR_TD))
  (bind (_outitem _outindex) (Utilities|Array|RandomArrayItemfromStream _l_ar_td _randomstream))
  (bind _swatchindex (Variables|RandomSwatch|GetSwatchIndex))
  (bind _returnvalue (+ _swatchindex 1))
  (LocalVariables|SetL_AR_TD TextureDataArray)
  (if (!= _outindex _swatchindex)
    (return _outitem)
    (elif (<= _returnvalue (- (Utilities|Array|Length (LocalVariables|GetL_AR_TD)) 1))
      (return (Utilities|Array|Get(acopy) (LocalVariables|GetL_AR_TD) _returnvalue))
      (else
        (return (Utilities|Array|Get(acopy) (LocalVariables|GetL_AR_TD)))))))

(fn ConstructionScript
  (bind _audiomap_0 (Variables|Default|GetAudioMap_0))
  (bind _slotmachinemusic (Variables|Default|GetSlotMachineMusic))
  (bind _soundloop (Variables|/Game/Sounds/Fort_Audio_Blueprints/PropAmbientAudioComponent/FortPropAudioComponent.FortPropAudioComponent_C|GetSoundLoop _slotmachinemusic))
  (bind _slotmachinemusic_2 (Variables|Default|GetSlotMachineMusic_2))
  (|CallParentFunction)
  (Default|GetRandomStream)
  (bind _randomizedtexturedata (Default|RandomizeSecondarySwatch (Variables|/Game/Environments/Apollo/Blueprints/RandomColor_Prop/RandomColor_Prop_Palette.RandomColor_Prop_Palette_C|GetSwatch_Palette (Variables|Default|GetcurrentPalette))))
  (Default|AssignNewMaterialstoSecondaryMesh _randomizedtexturedata (Variables|Default|GetSM_GoldWatch_PrizeMachine_A))
  (bind _output_get (Variables|Default|SetAR_TD_CC (Variables|/Game/Environments/Apollo/Blueprints/RandomColor_Prop/RandomColor_Prop_Palette.RandomColor_Prop_Palette_C|GetSwatch_Palette (Variables|Default|GetcurrentPalette))))
  (bind _returnvalue (Utilities|Array|RemoveItem _output_get _randomizedtexturedata))
  (if _returnvalue
    (bind _randomizedtexturedata_2 (Default|RandomizeSecondarySwatch (Variables|Default|GetAR_TD_CC)))
    (Default|AssignNewMaterialstoSecondaryMesh _randomizedtexturedata_2 (Variables|Default|GetSM_GoldWatch_PrizeMachine_A1))
    (if (.ReturnValue (Utilities|Map|Find _audiomap_0 _randomizedtexturedata))
      (if (not (== _soundloop (Utilities|Map|Find _audiomap_0 _randomizedtexturedata)))
        (Variables|/Game/Sounds/Fort_Audio_Blueprints/PropAmbientAudioComponent/FortPropAudioComponent.FortPropAudioComponent_C|SetSoundLoop (Utilities|Map|Find _audiomap_0 _randomizedtexturedata) _slotmachinemusic_2)
        (if (.ReturnValue (Utilities|Map|Find _audiomap_0 _randomizedtexturedata_2))
          (if (not (or (== _soundloop (Utilities|Map|Find _audiomap_0 _randomizedtexturedata_2)) (== (Variables|/Game/Sounds/Fort_Audio_Blueprints/PropAmbientAudioComponent/FortPropAudioComponent.FortPropAudioComponent_C|GetSoundLoop _slotmachinemusic_2) (Utilities|Map|Find _audiomap_0 _randomizedtexturedata_2))))
            (Variables|/Game/Sounds/Fort_Audio_Blueprints/PropAmbientAudioComponent/FortPropAudioComponent.FortPropAudioComponent_C|SetSoundLoop (Utilities|Map|Find _audiomap_0 _randomizedtexturedata_2) (Variables|Default|GetSlotMachineMusic_3))))))))
