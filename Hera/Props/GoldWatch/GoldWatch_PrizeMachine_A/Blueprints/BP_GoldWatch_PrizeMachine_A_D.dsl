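; Gold Watch prize machine prop Blueprint that varies screen materials and slot-machine music by colour swatch.
; When the swatch index is the palette's last, the construction script applies the first three palette entries to three machine meshes, using override, mapped or fallback materials.
; It then looks up audio loops for the first two entries, guarding against repeated loops.
; Location-seeded random-stream and secondary-swatch helpers are defined but not called here.
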(event EventBeginPlay
  (|CallParentFunction) ; This node is disabled and will not be called.\nDrag off pins to build functionality.
)

(event Collision|EventActorBeginOverlap
  (params (OtherActor :type Actor%20Object%20Reference))
  (|CallParentFunction OtherActor) ; This node is disabled and will not be called.\nDrag off pins to build functionality.
)

(event EventTick
  (params (DeltaSeconds :type Float%20%28single-precision%29))
  (|CallParentFunction DeltaSeconds) ; This node is disabled and will not be called.\nDrag off pins to build functionality.
)

(fn AssignNewMaterialsToSecondaryMesh
  (params (TextureData :type Building%20Texture%20Data%20Object%20Reference) (MeshComponent :type Primitive%20Component%20Object%20Reference))
  (bind _l_texturedata (LocalVariables|GetL_TextureData))
  (bind _materialmap (Variables|Default|GetMaterialMap))
  (bind (_value _returnvalue) (Utilities|Map|Find _materialmap _l_texturedata))
  (bind _l_texturedata_2 (LocalVariables|GetL_TextureData))
  (bind _overridematerial (Variables|/Script/FortniteGame.BuildingTextureData|GetOverrideMaterial _l_texturedata_2))
  (LocalVariables|SetL_TextureData TextureData)
  (if (Utilities|IsValid _overridematerial :!class /Script/Engine.KismetSystemLibrary)
    (Rendering|Material|SetMaterial MeshComponent 0 _overridematerial)
    (if _returnvalue
      (Rendering|Material|SetMaterial MeshComponent 1 _value)
      (else
        (Rendering|Material|SetMaterial MeshComponent 1 "/Hera/Props/GoldWatch/GoldWatch_PrizeMachine_A/Materials/MI_GoldWatch_PrizeMachineScreenDrop_A.MI_GoldWatch_PrizeMachineScreenDrop_A")))))

(fn GetRandomStream
  (bind _returnvalue (Math|Random|MakeRandomStream (Math|Float|Truncate (* (Class|EnvironmentArtFuncLib|RandomFloatbyLocation (Transformation|GetActorLocation) (Variables|RandomSwatch|GetSwatchTransitionDistance) true) 1e+04))))
  (Variables|Components|SetRandomStream _returnvalue))

(fn RandomizeSecondarySwatch
  (params (TextureDataArray :type Array%20of%20Building%20Texture%20Data%20Object%20References))
  (bind _randomstream (Variables|Components|GetRandomStream))
  (bind _l_ar_td (LocalVariables|GetL_AR_TD))
  (bind (_outitem _outindex) (Utilities|Array|RandomArrayItemfromStream _l_ar_td _randomstream))
  (bind _swatchindex (Variables|RandomSwatch|GetSwatchIndex))
  (bind _returnvalue (+ _swatchindex 1))
  (LocalVariables|SetL_AR_TD TextureDataArray)
  (if (!= _outindex _swatchindex)
    (return _outitem)
    (elif (<= _returnvalue (- (Utilities|Array|Length (LocalVariables|GetL_AR_TD)) 1))
      (return (Utilities|Array|Get(acopy) (LocalVariables|GetL_AR_TD) _returnvalue))
      (else
        (return (Utilities|Array|Get(acopy) (LocalVariables|GetL_AR_TD)))))))

(fn ConstructionScript
  (bind _currentpalette (Variables|Default|GetcurrentPalette))
  (bind _swatch_palette (Variables|/Game/Environments/Apollo/Blueprints/RandomColor_Prop/RandomColor_Prop_Palette.RandomColor_Prop_Palette_C|GetSwatch_Palette _currentpalette))
  (bind _output (Utilities|Array|Get(acopy) _swatch_palette 1))
  (bind _audiomap_0 (Variables|Default|GetAudioMap_0))
  (bind (_value _returnvalue) (Utilities|Map|Find _audiomap_0 _output))
  (bind _currentpalette_2 (Variables|Default|GetcurrentPalette))
  (bind _swatch_palette_2 (Variables|/Game/Environments/Apollo/Blueprints/RandomColor_Prop/RandomColor_Prop_Palette.RandomColor_Prop_Palette_C|GetSwatch_Palette _currentpalette_2))
  (bind _output_2 (Utilities|Array|Get(acopy) _swatch_palette_2))
  (bind (_value_2 _returnvalue_2) (Utilities|Map|Find _audiomap_0 _output_2))
  (bind _slotmachinemusic_2 (Variables|Default|GetSlotMachineMusic_2))
  (bind _slotmachinemusic (Variables|Default|GetSlotMachineMusic))
  (bind _soundloop (Variables|/Game/Sounds/Fort_Audio_Blueprints/PropAmbientAudioComponent/FortPropAudioComponent.FortPropAudioComponent_C|GetSoundLoop _slotmachinemusic))
  (|CallParentFunction)
  (if (== (Variables|RandomSwatch|GetSwatchIndex) (- (Utilities|Array|Length (Variables|/Game/Environments/Apollo/Blueprints/RandomColor_Prop/RandomColor_Prop_Palette.RandomColor_Prop_Palette_C|GetSwatch_Palette (Variables|Default|GetcurrentPalette))) 1))
    (Default|AssignNewMaterialstoSecondaryMesh _output_2 (Variables|Default|GetSM_GoldWatch_PrizeMachine_A))
    (Default|AssignNewMaterialstoSecondaryMesh _output (Variables|Default|GetSM_GoldWatch_PrizeMachine_A1))
    (Default|AssignNewMaterialstoSecondaryMesh (Utilities|Array|Get(acopy) (Variables|/Game/Environments/Apollo/Blueprints/RandomColor_Prop/RandomColor_Prop_Palette.RandomColor_Prop_Palette_C|GetSwatch_Palette (Variables|Default|GetcurrentPalette)) 2) (Variables|Default|GetSM_GoldWatch_PrizeMachine_A2))
    (if _returnvalue_2
      (if (not (== _soundloop _value_2))
        (Variables|/Game/Sounds/Fort_Audio_Blueprints/PropAmbientAudioComponent/FortPropAudioComponent.FortPropAudioComponent_C|SetSoundLoop _value_2 _slotmachinemusic_2)
        (if _returnvalue
          (if (not (or (== _soundloop _value) (== (Variables|/Game/Sounds/Fort_Audio_Blueprints/PropAmbientAudioComponent/FortPropAudioComponent.FortPropAudioComponent_C|GetSoundLoop _slotmachinemusic_2) _value)))
            (Variables|/Game/Sounds/Fort_Audio_Blueprints/PropAmbientAudioComponent/FortPropAudioComponent.FortPropAudioComponent_C|SetSoundLoop _value (Variables|Default|GetSlotMachineMusic_3))))))))
